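# Unlit material for the 3D fluid simulation that adds curl-noise and swirl forces to a velocity field stored as a sliced volume texture.
# Encoded volume coordinates, panned over time, sample a curl noise texture for turbulence.
# The swirl combines a tangent around the XY center with an inward pull, scaled by Z.
# The IsVelocity static switch adds both forces to the Source sample, or zero when off.
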Material /Volumetrics/Tools/FluidSimulation/Materials/3D/Forces/M_Fluid_Curl_Swirl_Force_01_Velocity
graph {
 properties {
  bAllowNegativeEmissiveColor=true
  blendMode=BLEND_AlphaComposite
  pixelDepthOffsetMode=PDOM_Legacy
  shadingModel=MSM_Unlit
  usedShadingModels=Unlit
 }
 nodes [
  n1 = TextureSampleParameter2D() {desc=Source, parameterName=Source, samplerType=SAMPLERTYPE_LinearColor, texture=/Volumetrics/Content/VolumeTextures/Textures/RT_Default}
  n2 = MaterialFunctionCall!(UVs, XY Frames=ComponentMask(None=VectorParameter() {parameterName=XYFrames, defaultValue=<rgba 16 16 0 0>}.RGB) {g=true, r=true}) {materialFunction=/Volumetrics/Content/VolumeTextures/MaterialFunctions/EncodeVolumeCoordinates}
  n3 = Multiply(
   A=MaterialFunctionCall(Speed=ScalarParameter() {parameterName=Constant Curl Pan Speed}) {materialFunction=&TimeWithSpeedVariable}.Frac Time
   B=VectorParameter() {parameterName=Pan Dir, defaultValue=<rgba 0 -0.05 0.1 0>}.RGB
  )
  n4 = TextureSample(UVs=Multiply(A=Add(A=@n2.Result, B=@n3), B=ScalarParameter() {defaultValue=1, parameterName=Curl Tiling})) {samplerType=SAMPLERTYPE_LinearColor, texture=/Volumetrics/Content/VolumeTextures/Textures/VT_CurlNoise}
  n5 = Multiply(A=@n4.RGB, B=ScalarParameter() {defaultValue=0.1, parameterName=Curl Force})
  n6 = MaterialFunctionCall(XY Frames=ComponentMask(None=VectorParameter() {parameterName=XYFrames, defaultValue=<rgba 16 16 0 0>}.RGB) {g=true, r=true}) {materialFunction=&EncodeVolumeCoordinates}
  n7 = Normalize(VectorInput=Subtract(A=ComponentMask(None=@n6.Result) {g=true, r=true}) {constB=0.5})
  n8 = MaterialFunctionCall(Float2=@n7) {materialFunction=&BreakOutFloat2Components}
  n9 = Add(
   A=Multiply(A=ScalarParameter() {defaultValue=2, parameterName=Swirl Towards Center}, B=Multiply(B=@n7) {constA=-1, constB=-2})
   B=MaterialFunctionCall(X=Multiply(A=@n8.G) {constB=-1}, Y=@n8.R) {materialFunction=&MakeFloat2}.Result
  )
  n10 = Constant()
  n11 = Multiply(
   A=Multiply(A=ScalarParameter() {desc="Firepit  = 1", parameterName=Swirl Force}, B=AppendVector(A=@n9, B=@n10))
   B=ComponentMask(None=@n6.Result) {b=true}
  )
  Add(A=@n1.RGB, B=StaticSwitchParameter(True=Add(A=@n5, B=@n11) {constB=0}, False=Constant()) {defaultValue=true, parameterName=IsVelocity})
  Multiply(A=@n4.RGB)
  n12 = TextureSampleParameter2D() {desc=Source, parameterName=Source, samplerType=SAMPLERTYPE_LinearColor, texture=&RT_Default}
  OneMinus(None=@n12.A)
 ]
}
