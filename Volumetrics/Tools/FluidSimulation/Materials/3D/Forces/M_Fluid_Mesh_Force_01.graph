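# Unlit alpha-composite material that writes mesh-driven forces and sources for a 3D fluid simulation, with negative values allowed.
# IsVelocity switches between a velocity pass and a density/temperature pass.
# Velocity adds Source, constant velocity and optional attraction plus swirl from a slice-interpolated jump-flood gradient volume, with inverse-square falloff and Color weighting.
# Density and temperature add exponentially saturated skeletal-mesh speed to Source, with density also masked by mesh alpha.
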Material /Volumetrics/Tools/FluidSimulation/Materials/3D/Forces/M_Fluid_Mesh_Force_01
graph {
 properties {
  bAllowNegativeEmissiveColor=true
  blendMode=BLEND_AlphaComposite
  pixelDepthOffsetMode=PDOM_Legacy
  shadingModel=MSM_Unlit
  usedShadingModels=Unlit
 }
 nodes [
  n1 = TextureCoordinate()
  n2 = ComponentMask(None=VectorParameter() {parameterName=XYFrames, defaultValue=<rgba 16 16 0 0>}.RGB) {g=true, r=true}
  n3 = MaterialFunctionCall!(
   Texture Object
   Position=Multiply(A=MaterialFunctionCall!(UVs=@n1, XY Frames=@n2) {materialFunction=/Volumetrics/Content/VolumeTextures/MaterialFunctions/EncodeVolumeCoordinates}.Result)
   Scale=Constant() {r=1}
   Num Frames=ComponentMask(None=VectorParameter() {parameterName=SampleXYFrames, defaultValue=<rgba 16 8 0 0>}.RGB) {g=true, r=true}
   Clamp Z=StaticBool() {value=true}
  ) {materialFunction=/Volumetrics/Content/VolumeTextures/MaterialFunctions/VolumeTextureFunction}
  n4 = TextureSampleParameter2D!(UVs=@n3.UV 1, Level, Apply View MipBias) {constMipValue=0, mipValueMode=TMVM_MipLevel, parameterName=Gradient, texture=/Volumetrics/Tools/JumpFlooding/Textures/RT_JumpFlood_01}
  n5 = TextureSampleParameter2D(UVs=@n3.UV 2) {constMipValue=0, mipValueMode=TMVM_MipLevel, parameterName=Gradient, texture=&RT_JumpFlood_01}
  n6 = LinearInterpolate(A=@n4.RGB, B=@n5.RGB, Alpha=@n3.Z Phase)
  n7 = Constant3Vector() {constant=<rgba 0 0 1 0>}
  n8 = LinearInterpolate(A=@n4.A, B=@n5.A, Alpha=@n3.Z Phase)
  n9 = ScalarParameter() {defaultValue=0.5, parameterName=Center}
  n10 = Abs(None=Multiply(A=Subtract(A=@n8, B=@n9)) {constB=2})
  n11 = ScalarParameter() {defaultValue=0.1, parameterName=MeshMin}
  n12 = Max(A=@n10, B=@n11) {constB=0.1}
  n13 = Multiply(A=Multiply(A=@n12, B=@n12), B=If(A=@n10, B=Multiply(A=@n11) {constB=-1}, "A > B"=Constant() {r=1}, "A < B"=Constant() {r=-1}))
  n14 = Divide(A=Multiply(A=ScalarParameter() {parameterName=DistanceFieldSwirl}, B=CrossProduct(A=@n6, B=@n7)), B=@n13)
  n15 = Divide(A=@n6, B=@n13)
  n16 = TextureSampleParameter2D!(UVs, Apply View MipBias) {desc=Color, parameterName=Color, samplerType=SAMPLERTYPE_LinearColor, texture=/Volumetrics/Content/VolumeTextures/Textures/RT_Default}
  n17 = MaterialFunctionCall(f3=@n16.RGB) {materialFunction=&AddComponents}
  n18 = Multiply(
   A=Multiply(A=Add(A=@n14, B=@n15))
   B=Multiply(A=Divide(A=@n17.f3) {constB=1000}, B=ScalarParameter() {parameterName=DistanceFieldAttraction})
  )
  n19 = TextureSampleParameter2D() {desc=Source, parameterName=Source, samplerType=SAMPLERTYPE_LinearColor, texture=&RT_Default}
  n20 = Add(
   A=VectorParameter() {parameterName=Constant Velocity}.RGB
   B=StaticSwitchParameter(True=Add(A=@n18, B=@n19.RGB), False=@n19.RGB) {defaultValue=true, parameterName=DistanceFieldAttractor}
  )
  n21 = TextureSampleParameter2D() {
   desc=Velocity
   parameterName=SkeletalMeshTexture
   samplerType=SAMPLERTYPE_LinearColor
   texture=&RT_Default
  }
  n22 = MaterialFunctionCall(Vector 3=@n21.RGB) {materialFunction=/Engine/Functions/Engine_MaterialFunctions02/Utility/VectorLength}
  n23 = ScalarParameter() {defaultValue=-2, parameterName=MinVelocityForDensity}
  MaterialExpressionMax_3 = Max(A=Multiply(A=Subtract(A=@n22.V3 Length, B=@n23), B=ScalarParameter() {defaultValue=0.25, parameterName=MeshVelocityScale})) {constB=0}
  n24 = Custom!(x=@MaterialExpressionMax_3) {
   code="return 1-exp(-x);"
   description=exp
   outputType=CMOT_Float1
   inputs [{inputName=x, input {expression=&self:MaterialExpressionMax_3}}]
  }
  n25 = Multiply(A=@n21.A, B=ScalarParameter() {parameterName=SkeletalMeshDensity})
  n26 = Multiply(A=Add(A=@n19.RGB, B=Multiply(A=@n24, B=@n25)))
  n27 = TextureSampleParameter2D() {desc=Source, parameterName=Source, samplerType=SAMPLERTYPE_LinearColor, texture=&RT_Default}
  n28 = ScalarParameter() {parameterName=MinMeshVelocity}
  MaterialExpressionMax_4 = Max(A=Multiply(A=Subtract(A=@n22.V3 Length, B=@n28), B=ScalarParameter() {defaultValue=0.25, parameterName=MeshVelocityScale})) {constB=0}
  n29 = Custom(x=@MaterialExpressionMax_4) {
   code="return 1-exp(-x);"
   description=exp
   outputType=CMOT_Float1
   inputs [{inputName=x, input {expression=&self:MaterialExpressionMax_4}}]
  }
  n30 = Add(A=@n27.A, B=Multiply(A=@n29, B=ScalarParameter() {parameterName=VelocityToTemperature}))
  n31 = Clamp(None=AppendVector(
    A=StaticSwitchParameter(True=@n20, False=@n26) {defaultValue=true, parameterName=IsVelocity}
    B=StaticSwitchParameter(True=@n27.A, False=@n30) {defaultValue=true, parameterName=IsVelocity}
   )) {maxDefault=40000, minDefault=-40000}
  ComponentMask(None=@n31) {b=true, g=true, r=true}
  OneMinus(None=ComponentMask(None=@n31) {a=true})
  OneMinus(None=@n21.A)
  n32 = Subtract(A=ScalarParameter() {defaultValue=0.5, parameterName=Center}, B=ScalarParameter() {defaultValue=0.001, parameterName=Core})
  Saturate(None=Divide(A=Subtract(A=@n8, B=@n32), B=ScalarParameter() {defaultValue=0.005, parameterName=Mask Width}))
  SphereMask(
   A=@n8
   B=ScalarParameter() {defaultValue=0.505, parameterName=DF Density Location}
   Radius=ScalarParameter() {defaultValue=0.0025, parameterName=DF Density Radius}
  ) {hardnessPercent=50}
 ]
}
